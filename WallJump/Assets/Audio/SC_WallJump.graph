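# Sound cue for the wall jump kick-off whoosh, played as player foley with distance attenuation.
# An IsLocal branch plays a much quieter mix for the local player; otherwise an IsEnemy branch plays full volume when true or unset and silence when false.
# Each audible path randomly picks one of four whoosh waves with pitch and volume variation.
# A concurrency override caps simultaneous instances at two, and virtualization is disabled.
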SoundCue /WallJump/Assets/Audio/SC_WallJump
root=Branch_0
nodes [
 {id=Branch_0, children [Mixer_1,Branch_1,Branch_1], properties {boolParameterName=IsLocal}}
 {id=Branch_1, children [Mixer_2,null,Mixer_2], properties {boolParameterName=IsEnemy}}
 {id=Mixer_1, children [Modulator_4], properties {inputVolume [0.1]}}
 {id=Mixer_2, children [Modulator_2], properties {inputVolume [1]}}
 {id=Modulator_2, children [Random_1], properties {pitchMax=1.2, pitchMin=0.85, volumeMax=1, volumeMin=0.8}}
 {id=Modulator_4, children [Random_4], properties {pitchMax=1.2, pitchMin=0.85, volumeMax=1, volumeMin=0.8}}
 {
  id=Random_1
  children [WavePlayer_0, WavePlayer_5, WavePlayer_6, WavePlayer_16]
  properties {bShouldExcludeFromBranchCulling=true, weights [1,1,1,1]}
 }
 {
  id=Random_4
  children [WavePlayer_17, WavePlayer_18, WavePlayer_19, WavePlayer_20]
  properties {bShouldExcludeFromBranchCulling=true, weights [1,1,1,1]}
 }
 {id=WavePlayer_0, properties {soundWaveAssetPtr=&/Wall_Jump_Kick_Off_Whoosh_01}}
 {id=WavePlayer_16, properties {soundWaveAssetPtr=&/Wall_Jump_Kick_Off_Whoosh_04}}
 {id=WavePlayer_17, properties {soundWaveAssetPtr=&Wall_Jump_Kick_Off_Whoosh_01}}
 {id=WavePlayer_18, properties {soundWaveAssetPtr=&/Wall_Jump_Kick_Off_Whoosh_02}}
 {id=WavePlayer_19, properties {soundWaveAssetPtr=&/Wall_Jump_Kick_Off_Whoosh_03}}
 {id=WavePlayer_20, properties {soundWaveAssetPtr=&Wall_Jump_Kick_Off_Whoosh_04}}
 {id=WavePlayer_5, properties {soundWaveAssetPtr=&Wall_Jump_Kick_Off_Whoosh_02}}
 {id=WavePlayer_6, properties {soundWaveAssetPtr=&Wall_Jump_Kick_Off_Whoosh_03}}
]
properties {
 attenuationSettings=&/ATT_WallJump_Whoosh
 bOverrideConcurrency=true
 maxDistance=3584
 soundClassObject=&Fort_Player_Foley_Athena
 virtualizationMode=Disabled
 volumeMultiplier=1
 concurrencyOverrides {maxCount=2}
}
