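# Sound cue for the wall jump kick-off on sand, playing one of four random dirt kick-off samples through a shared modulator.
# IsLocal and IsEnemy branches pick first-person, third-person friendly or third-person enemy foley sound classes, with enemy as the fallback when IsEnemy is unset.
# Each route has its own mixer gain, loudest for local and quietest for friendly.
# Wall jump attenuation and a concurrency cap limit playback, with virtualization disabled.
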SoundCue /WallJump/Assets/Audio/SC_WallJump_Sand
root=Branch_0
nodes [
 {id=Branch_0, children [SoundClass_0,Branch_1,Branch_1], properties {boolParameterName=IsLocal}}
 {id=Branch_1, children [SoundClass_2,SoundClass_1,SoundClass_2], properties {boolParameterName=IsEnemy}}
 {id=Mixer_0, children [Modulator_0], properties {inputVolume [2]}}
 {id=Mixer_1, children [Modulator_0], properties {inputVolume [1]}}
 {id=Mixer_2, children [Modulator_0], properties {inputVolume [0.3]}}
 {id=Modulator_0, children [Random_1], properties {volumeMax=1}}
 {
  id=Random_1
  children [WavePlayer_4,WavePlayer_5,WavePlayer_6,WavePlayer_7]
  properties {bShouldExcludeFromBranchCulling=true, weights [1,1,1,1]}
 }
 {id=SoundClass_0, children [Mixer_0], properties {soundClassOverride=&Fort_Player_Foley_1P_Athena}}
 {
  id=SoundClass_1
  children [Mixer_2]
  properties {soundClassOverride=/Game/Sounds/FortSoundClassesAndModes/Fort_Player_Foley_3P_Friendly_Athena}
 }
 {id=SoundClass_2, children [Mixer_1], properties {soundClassOverride=&Fort_Player_Foley_3P_Enemy_Athena}}
 {id=WavePlayer_4, properties {soundWaveAssetPtr=&/Wall_Jump_Kick_Off_Dirt_01}}
 {id=WavePlayer_5, properties {soundWaveAssetPtr=&/Wall_Jump_Kick_Off_Dirt_02}}
 {id=WavePlayer_6, properties {soundWaveAssetPtr=&/Wall_Jump_Kick_Off_Dirt_03}}
 {id=WavePlayer_7, properties {soundWaveAssetPtr=&/Wall_Jump_Kick_Off_Dirt_04}}
]
properties {
 attenuationSettings=&/ATT_WallJump
 bOverrideConcurrency=true
 maxDistance=3584
 virtualizationMode=Disabled
 volumeMultiplier=1
 concurrencyOverrides {maxCount=4}
}
